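; Gameplay cue actor that fades the Marksman visor's scanning effect in and out on a Fort player pawn.
; On becoming relevant it captures the visor mesh through a cosmetic-part callback and starts ticking.
; Tick ramps ActivePercent over the ScanFootprintSeconds scalable float, driving the ScreenAlpha and ActivateScanningAlpha material parameters.
; Ceasing relevance reverses the ramp; tick stops at either limit, and the cue recycles after fading out.
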(event EventTick
  (params (DeltaSeconds :type Float%20%28single-precision%29))
  (Variables|Default|SetActivePercent (Math|Float|Clamp(Float) (Math|Float|SelectFloat (+ (Variables|Default|GetActivePercent) (/ DeltaSeconds (Math|Float|Max(Float) (ScalableFloat|GetValueatLevel (Variables|Default|GetScanFootprintSeconds)) 0.001))) (- (Variables|Default|GetActivePercent) (/ DeltaSeconds (Math|Float|Max(Float) (ScalableFloat|GetValueatLevel (Variables|Default|GetScanFootprintSeconds)) 0.001))) (Variables|Default|GetIsActive))))
  (CallFunction|UpdateforActivePercent)
  (if (Variables|Default|GetIsActive)
    (if (>= (Variables|Default|GetActivePercent) 1.0)
      (Actor|Tick|SetActorTickEnabled false))
    (else
      (if (<= (Variables|Default|GetActivePercent) 0)
        (Actor|Tick|SetActorTickEnabled false)
        (GameplayCueNotify|End%28Recycle%29GameplayCue)))))

(fn OnCeaseRelevant
  (params (MyTarget :type Actor%20Object%20Reference) (Parameters :type Gameplay%20Cue%20Parameters%20Structure%20%28by%20ref%29))
  (|CallParentFunction MyTarget Parameters)
  (if (Variables|Default|GetIsActive)
    (Variables|Default|SetIsActive false)
    (Actor|Tick|SetActorTickEnabled true)
    (return false)))

(fn OnVisorMeshPartLoaded
  (params (Pawn :type Fort%20Player%20Pawn%20Object%20Reference) (Part :type Custom%20Character%20Part%20Object%20Reference) (MeshPart :type Skeletal%20Mesh%20Component%20Object%20Reference) (PartType :type EFortCustomPartType%20Enum :default "Head"))
  (if (== Pawn (Variables|Default|GetCurrentTargetPawn))
    (Variables|Default|SetVisorMesh MeshPart)))

(fn UpdateForActivePercent
  (bind _activepercent (Variables|Default|GetActivePercent))
  (bind _visormesh (Variables|Default|GetVisorMesh)
    (:then
      (Rendering|Material|SetScalarParameterValueonMaterials _visormesh "ScreenAlpha" _activepercent)
      (Rendering|Material|SetScalarParameterValueonMaterials _visormesh "ActivateScanningAlpha" _activepercent))
    (:else)))

(fn ConstructionScript)

(fn OnBecomeRelevant
  (params (MyTarget :type Actor%20Object%20Reference) (Parameters :type Gameplay%20Cue%20Parameters%20Structure%20%28by%20ref%29))
  (|CallParentFunction MyTarget Parameters)
  (bind _asfort_player_pawn (Utilities|Casting|CastToFortPlayerPawn MyTarget))
  (Variables|Default|SetCurrentTargetPawn _asfort_player_pawn)
  (Variables|Default|SetVisorMesh 0)
  (Cosmetics|WhenCosmeticSpecificPartApplied (Variables|Default|GetCurrentTargetPawn) (EventDispatchers|CreateEvent :!event OnVisorMeshPartLoaded) true "Gameplay")
  (Variables|Default|SetIsActive true)
  (Actor|Tick|SetActorTickEnabled true)
  (return false))
